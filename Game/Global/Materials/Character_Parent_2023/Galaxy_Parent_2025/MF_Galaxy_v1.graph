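# Material function that layers an animated deep-space galaxy effect onto a character material, apparently confined by a selectable TechArtMask channel.
# Two scrolling nebula layers use triplanar virtual-box projection for depth, then hue shift and HDR remapping, and are screen-blended.
# Two dither-sampled star layers twinkle through MF_RandomBlink and are masked by nebula luminance.
# The time-of-day-tinted result feeds emissive, with optional overrides for base color, metallic, specular and roughness.
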MaterialFunction /Game/Global/Materials/Character_Parent_2023/Galaxy_Parent_2025/MF_Galaxy_v1
graph {nodes [
  FunctionOutput(None=BlendMaterialAttributes(A=NamedRerouteUsage(), B=NamedRerouteUsage(), Alpha=NamedRerouteUsage()))
  n1 = TextureObjectParameter() {
   group=Galaxy - Stars
   parameterName=Galaxy.StarsTexture
   samplerType=SAMPLERTYPE_LinearColor
   sortPriority=0
   texture=/Game/Global/Materials/Character_Parent_2023/Galaxy_Parent_2025/Textures/T_GalaxyStars
  }
  NamedRerouteDeclaration(None=@n1) {name=StarTex, nodeColor=<rgba 1 0.1641 0 1>}
  n2 = GetMaterialAttributes(None=NamedRerouteUsage()) {attributeGetTypes [
    #1
    #2
    #3
    #4
    #5
   ]}
  n3 = ScalarParameter() {group=Galaxy - General Parameters, parameterName=Galaxy.BaseMaterial_SMR_Override}
  NamedRerouteDeclaration(None=LinearInterpolate(
    A=@n2.Roughness
    B=ScalarParameter() {defaultValue=0.2, group=Galaxy - General Parameters, parameterName=Galaxy.Roughnes_Override, sortPriority=30}
    Alpha=@n3
   ) {constB=0}) {name=OUT_Roughness, nodeColor=<rgba 0 0.4844 1 1>}
  NamedRerouteDeclaration(None=LinearInterpolate(
    A=@n2.Metallic
    B=ScalarParameter() {defaultValue=0.1, group=Galaxy - General Parameters, parameterName=Galaxy.Metallic_Override, sortPriority=30}
    Alpha=@n3
   ) {constB=0}) {name=OUT_Metallic, nodeColor=<rgba 0 0.4844 1 1>}
  n4 = NamedRerouteUsage()
  n5 = NamedRerouteUsage()
  n6 = NamedRerouteUsage()
  n7 = ScalarParameter() {defaultValue=0.01, group=Galaxy - Stars, parameterName=Galaxy.Stars_TwinkleFrequency_Layer1, sortPriority=15}
  n8 = ScalarParameter() {defaultValue=0.7, group=Galaxy - Stars, parameterName=Galaxy.Stars_TwinkleSpeed_Layer1, sortPriority=16}
  n9 = Max(
   A=Multiply(
    A=MaterialFunctionCall!(Mask=@n4, UVs=@n5, RandomOffset=@n6, BlinkFrequency=@n7, BlinkSpeed=@n8) {materialFunction=/Game/Global/Materials/Character_Parent_2023/RandomBlink_v1/MF_RandomBlink}.Result
    B=ScalarParameter() {defaultValue=5, group=Galaxy - Stars, parameterName=Galaxy.Stars_TwinkleBrightness_Layer1, sortPriority=17}
   )
   B=NamedRerouteUsage()
  )
  n10 = ScalarParameter() {defaultValue=1, group=Galaxy - Stars, parameterName=Galaxy.Stars_Brightness_Layer1, sortPriority=14}
  n11 = NamedRerouteUsage()
  n12 = NamedRerouteUsage()
  n13 = NamedRerouteUsage()
  n14 = ScalarParameter() {defaultValue=0.02, group=Galaxy - Stars, parameterName=Galaxy.Stars_TwinkleFrequency_Layer2, sortPriority=25}
  n15 = ScalarParameter() {defaultValue=0.5, group=Galaxy - Stars, parameterName=Galaxy.Stars_TwinkleSpeed_Layer2, sortPriority=26}
  n16 = Max(
   A=Multiply(
    A=MaterialFunctionCall(Mask=@n11, UVs=@n12, RandomOffset=@n13, BlinkFrequency=@n14, BlinkSpeed=@n15) {materialFunction=&MF_RandomBlink}.Result
    B=ScalarParameter() {defaultValue=5, group=Galaxy - Stars, parameterName=Galaxy.Stars_TwinkleBrightness_Layer2, sortPriority=27}
   )
   B=NamedRerouteUsage()
  )
  n17 = ScalarParameter() {defaultValue=1, group=Galaxy - Stars, parameterName=Galaxy.Stars_Brightness_Layer2, sortPriority=24}
  NamedRerouteDeclaration(None=Max(A=Multiply(A=@n9, B=@n10), B=Multiply(A=@n16, B=@n17))) {name=Stars, nodeColor=<rgba 0 1 0.5078 1>}
  n18 = TextureObjectParameter() {
   group=Galaxy - Nebula
   parameterName=Galaxy.NebulaTexture_Layer1
   sortPriority=10
   texture=/Game/Global/Materials/Character_Parent_2023/Galaxy_Parent_2025/Textures/T_FN_Nebula_style_blue
  }
  n19 = ScalarParameter() {defaultValue=100, group=Galaxy - Nebula, parameterName=Galaxy.Nebula_Depth_Layer1, sortPriority=11}
  n20 = ScalarParameter() {defaultValue=0.4, group=Galaxy - Nebula, parameterName=Galaxy.Nebula_Tile_Layer1, sortPriority=12}
  n21 = ScalarParameter() {defaultValue=0.4, group=Galaxy - Nebula, parameterName=Galaxy.Nebula_BlendContrast_Layer1, sortPriority=13}
  n22 = VectorParameter() {group=Galaxy - Nebula, parameterName=Galaxy.Nebula_Speed_Layer1, sortPriority=14, defaultValue=<rgba 0.002 -0.002 0 1>}
  n23 = ComponentMask(None=@n22.RGB) {g=true, r=true}
  n24 = VectorParameter() {group=Galaxy - General Parameters, parameterName=Galaxy.OriginOffset, sortPriority=10, defaultValue=<rgba 0 0 0 1>}
  n25 = NamedRerouteDeclaration(None=@n24.RGB) {name=ActorOffset, nodeColor=<rgba 0.9688 1 0 1>}
  n26 = Power(
   Base=MaterialFunctionCall(
    Hue Shift Percentage=ScalarParameter() {group=Galaxy - Nebula, parameterName=Galaxy.Nebula_HueShift_Layer1, sortPriority=17}
    Texture=MaterialFunctionCall!(Texture=@n18, Depth=@n19, Tile=@n20, BlendContrast=@n21, SpeedXY=@n23, ActorOffset=@n25, bUseLocalCameraVector, bUseDitherSingleSample) {materialFunction=/Game/Global/MaterialFunctions/MF_TriplanarVirtualBox}.Result
   ) {materialFunction=&HueShift}.Result
   Exp=ScalarParameter() {defaultValue=1.5, group=Galaxy - Nebula, parameterName=Galaxy.Nebula_Power_Layer1, sortPriority=16}
  )
  n27 = ScalarParameter() {defaultValue=0.35, group=Galaxy - Nebula, parameterName=Galaxy.Nebula_Desaturation_Layer1, sortPriority=18}
  NamedRerouteDeclaration(None=MaterialFunctionCall(
    LDR Input=Desaturation(None=@n26, Fraction=@n27)
    HDR Intensity Multiplier=ScalarParameter() {defaultValue=0.6, group=Galaxy - Nebula, parameterName=Galaxy.Nebula_Brightness_Layer1, sortPriority=15}
   ) {materialFunction=&DeriveHDRfromLDR}.RemappedRange) {name=Galaxy_01, nodeColor=<rgba 1 0.2578 0 1>}
  NamedRerouteDeclaration(None=LinearInterpolate(
    A=@n2.Specular
    B=ScalarParameter() {defaultValue=0.5, group=Galaxy - General Parameters, parameterName=Galaxy.Specular_Override, sortPriority=30}
    Alpha=@n3
   ) {constB=0}) {name=OUT_Spec, nodeColor=<rgba 0 0.4844 1 1>}
  NamedRerouteDeclaration(None=@n2.EmissiveColor) {name=IN_Emissive, nodeColor=<rgba 0.4766 1 0 1>}
  n28 = VectorParameter() {group=Galaxy - Stars, parameterName=Glaxy.Stars_Speed_Layer1, sortPriority=13, defaultValue=<rgba -0.001 -0.01 0 1>}
  MaterialExpressionMaterialFunctionCall_11 = MaterialFunctionCall(
   Texture=NamedRerouteUsage()
   Depth=ScalarParameter() {defaultValue=75, group=Galaxy - Stars, parameterName=Galaxy.Stars_Depth_Layer1, sortPriority=10}
   Tile=ScalarParameter() {defaultValue=1.5, group=Galaxy - Stars, parameterName=Galaxy.Stars_Tile_Layer1, sortPriority=11}
   BlendContrast=ScalarParameter() {defaultValue=1, group=Galaxy - Stars, parameterName=Galaxy.Stars_BlendContrast_Layer1, sortPriority=12}
   SpeedXY=ComponentMask(None=@n28.RGB) {g=true, r=true}
   ActorOffset=NamedRerouteUsage()
   bUseDitherSingleSample=StaticBool() {value=true}
  ) {materialFunction=&MF_TriplanarVirtualBox}
  NamedRerouteDeclaration(None=@MaterialExpressionMaterialFunctionCall_11.DitherUVs) {name=CloseStarsUVW, nodeColor=<rgba 1 0 0.7969 1>}
  n29 = Convert(None=@MaterialExpressionMaterialFunctionCall_11.Result) {
   convertInputs [{
     type=Vector3
     expressionInput {
      expression=&self:MaterialExpressionMaterialFunctionCall_11
      inputName=None
      mask=0
      maskA=0
      maskB=0
      maskG=0
      maskR=0
      outputIndex=0
     }
    }]
   convertOutputs [{type=Scalar}, {type=Scalar}, {type=Scalar}]
  }
  NamedRerouteDeclaration(None=NamedRerouteDeclaration(None=@n29) {name=CloseStarsB, nodeColor=<rgba 0.4766 1 0 1>}) {name=CloseStarsRandom, nodeColor=<rgba 1 0.3984 0 1>}
  NamedRerouteDeclaration(None=NamedRerouteUsage()) {name=OUT_Emissive, nodeColor=<rgba 1 0 0.1875 1>}
  n30 = VectorParameter() {group=Galaxy - General Parameters, parameterName=Galaxy.BaseColor_Override, sortPriority=30, defaultValue=<rgba 0 0 0 1>}
  NamedRerouteDeclaration(None=LinearInterpolate(A=@n2.BaseColor, B=@n30.RGB, Alpha=ScalarParameter() {group=Galaxy - General Parameters, parameterName=Galaxy.BaseMaterial_Color_Override}) {constB=0}) {name=OUT_BaseColor, nodeColor=<rgba 1 0 0.6562 1>}
  n31 = VectorParameter() {group=Galaxy - Stars, parameterName=Glaxy.Stars_Speed_Layer2, sortPriority=23, defaultValue=<rgba -0.001 -0.01 0 1>}
  MaterialExpressionMaterialFunctionCall_4 = MaterialFunctionCall(
   Texture=NamedRerouteUsage()
   Depth=ScalarParameter() {defaultValue=150, group=Galaxy - Stars, parameterName=Galaxy.Stars_Depth_Layer2, sortPriority=20}
   Tile=ScalarParameter() {defaultValue=3, group=Galaxy - Stars, parameterName=Galaxy.Stars_Tile_Layer2, sortPriority=21}
   BlendContrast=ScalarParameter() {defaultValue=1, group=Galaxy - Stars, parameterName=Galaxy.Stars_BlendContrast_Layer2, sortPriority=22}
   SpeedXY=ComponentMask(None=@n31.RGB) {g=true, r=true}
   ActorOffset=NamedRerouteUsage()
   bUseDitherSingleSample=StaticBool() {value=true}
  ) {materialFunction=&MF_TriplanarVirtualBox}
  n32 = Convert(None=@MaterialExpressionMaterialFunctionCall_4.Result) {
   convertInputs [{
     type=Vector3
     expressionInput {
      expression=&self:MaterialExpressionMaterialFunctionCall_4
      inputName=None
      mask=0
      maskA=0
      maskB=0
      maskG=0
      maskR=0
      outputIndex=0
     }
    }]
   convertOutputs [{type=Scalar}, {type=Scalar}, {type=Scalar}]
  }
  NamedRerouteDeclaration(None=NamedRerouteDeclaration(None=@n32) {name=FarStarsB, nodeColor=<rgba 0 0.4375 1 1>}) {name=FarStarsRandom, nodeColor=<rgba 1 0.3984 0 1>}
  n33 = FunctionInput() {bUsePreviewValueAsDefault=true, inputName=TechArtMask, sortPriority=1, previewValue=<1 1 1 1>}
  NamedRerouteDeclaration(None=ChannelMaskParameter(None=@n33) {group=Galaxy - General Parameters, parameterName=Galaxy.MaskChannel, sortPriority=0}.R) {name=GalaxyMask, nodeColor=<rgba 0 1 0.6953 1>}
  NamedRerouteDeclaration(None=@n32) {name=FarStarsR, nodeColor=<rgba 0 1 0.2031 1>}
  NamedRerouteDeclaration(None=@n32) {name=FarStarsG, nodeColor=<rgba 1 0 0.1172 1>}
  NamedRerouteDeclaration(None=@MaterialExpressionMaterialFunctionCall_4.DitherUVs) {name=FarStarsUVW, nodeColor=<rgba 1 0 0.8906 1>}
  NamedRerouteDeclaration(None=@n29) {name=CloseStarsR, nodeColor=<rgba 0 0.3437 1 1>}
  n34 = SetMaterialAttributes!(
   MaterialAttributes=NamedRerouteDeclaration(None=FunctionInput() {inputName=Material, inputType=FunctionInput_MaterialAttributes}) {name=IN_Material, nodeColor=<rgba 1 0 0.9609 1>}
   BaseColor=NamedRerouteUsage()
   Metallic=NamedRerouteUsage()
   Specular=NamedRerouteUsage()
   Roughness=NamedRerouteUsage()
   EmissiveColor=Add(A=NamedRerouteUsage(), B=NamedRerouteUsage())
  ) {attributeSetTypes [
    #5
    #2
    #3
    #1
    #4
   ]}
  NamedRerouteDeclaration(None=@n34) {name=OUT_Material, nodeColor=<rgba 0 1 0.625 1>}
  NamedRerouteDeclaration(None=@n29) {name=CloseStarsG, nodeColor=<rgba 0 0.1562 1 1>}
  n35 = TextureObjectParameter() {
   group=Galaxy - Nebula
   parameterName=Galaxy.NebulaTexture_Layer2
   sortPriority=20
   texture=/Game/Global/Materials/Character_Parent_2023/Galaxy_Parent_2025/Textures/T_FN_Nebula_style2
  }
  n36 = ScalarParameter() {defaultValue=50, group=Galaxy - Nebula, parameterName=Galaxy.Nebula_Depth_Layer2, sortPriority=21}
  n37 = ScalarParameter() {defaultValue=0.4, group=Galaxy - Nebula, parameterName=Galaxy.Nebula_Tile_Layer2, sortPriority=22}
  n38 = ScalarParameter() {defaultValue=1, group=Galaxy - Nebula, parameterName=Galaxy.Nebula_BlendContrast_Layer2, sortPriority=23}
  n39 = VectorParameter() {group=Galaxy - Nebula, parameterName=Galaxy.Nebula_Speed_Layer2, sortPriority=24, defaultValue=<rgba -0.001 0.0005 0 1>}
  n40 = ComponentMask(None=@n39.RGB) {g=true, r=true}
  n41 = NamedRerouteUsage()
  n42 = Power(
   Base=MaterialFunctionCall(
    Hue Shift Percentage=ScalarParameter() {defaultValue=0.05, group=Galaxy - Nebula, parameterName=Galaxy.Nebula_HueShift_Layer2, sortPriority=27}
    Texture=MaterialFunctionCall(Texture=@n35, Depth=@n36, Tile=@n37, BlendContrast=@n38, SpeedXY=@n40, ActorOffset=@n41) {materialFunction=&MF_TriplanarVirtualBox}.Result
   ) {materialFunction=&HueShift}.Result
   Exp=ScalarParameter() {defaultValue=0.85, group=Galaxy - Nebula, parameterName=Galaxy.Nebula_Power_Layer2, sortPriority=26}
  )
  n43 = ScalarParameter() {group=Galaxy - Nebula, parameterName=Galaxy.Nebula_Desaturation_Layer2, sortPriority=28}
  NamedRerouteDeclaration(None=MaterialFunctionCall(
    LDR Input=Desaturation(None=@n42, Fraction=@n43)
    HDR Intensity Multiplier=ScalarParameter() {defaultValue=0.5, group=Galaxy - Nebula, parameterName=Galaxy.Nebula_Brightness_Layer2, sortPriority=25}
   ) {materialFunction=&DeriveHDRfromLDR}.RemappedRange) {name=Galaxy_02, nodeColor=<rgba 0.0781 0 1 1>}
  n44 = Constant() {r=1}
  n45 = Multiply(A=Subtract(A=Constant() {r=1}, B=NamedRerouteUsage()), B=Subtract(A=Constant() {r=1}, B=NamedRerouteUsage()))
  n46 = Desaturation(None=NamedRerouteDeclaration(None=Subtract(A=@n44, B=@n45)) {name=Galaxies, nodeColor=<rgba 1 0 0.2578 1>}, Fraction=Constant() {r=1})
  n47 = ScalarParameter() {defaultValue=0.2, group=Galaxy - Stars, parameterName=Galaxy.Stars_NebulaMask_Power, sortPriority=3}
  NamedRerouteDeclaration(None=MaterialFunctionCall(
    In=Power(Base=@n46, Exp=@n47)
    Contrast=ScalarParameter() {defaultValue=1, group=Galaxy - Stars, parameterName=Galaxy.Stars_NebulaMask_Contrast, sortPriority=4}
   ) {materialFunction=&CheapContrast}.Result) {name=StarMask, nodeColor=<rgba 0.9922 1 0 1>}
  n48 = NamedRerouteUsage()
  n49 = NamedRerouteUsage()
  n50 = VectorParameter() {group=Galaxy - General Parameters, parameterName=Galaxy.TOD_Morning, sortPriority=50, defaultValue=<rgba 1 1 1 1>}
  n51 = VectorParameter() {group=Galaxy - General Parameters, parameterName=Galaxy.TOD_Day, sortPriority=51, defaultValue=<rgba 1 1 1 1>}
  n52 = VectorParameter() {group=Galaxy - General Parameters, parameterName=Galaxy.TOD_Evening, sortPriority=52, defaultValue=<rgba 1 1 1 1>}
  n53 = VectorParameter() {group=Galaxy - General Parameters, parameterName=Galaxy.TOD_Night, sortPriority=53, defaultValue=<rgba 1 1 1 1>}
  n54 = Multiply(
   A=NamedRerouteUsage()
   B=MaterialFunctionCall(Color_Morning=@n50.RGB, Color_Day=@n51.RGB, Color_Evening=@n52.RGB, Color_Night=@n53.RGB) {materialFunction=&Color_Time_of_Day}.Result
  )
  n55 = Multiply(
   A=Saturate(None=LinearInterpolate(A=@n48, Alpha=@n49) {constB=0})
   B=Multiply(A=@n54, B=ScalarParameter() {defaultValue=25, group=Galaxy - Stars, parameterName=Galaxy.Stars_GlobalBrightness, sortPriority=1}) {constB=20}
  )
  n56 = ScalarParameter() {defaultValue=0.5, group=Galaxy - Stars, parameterName=Galaxy.Stars_Desaturation, sortPriority=2}
  NamedRerouteDeclaration(None=Add(A=Desaturation(None=@n55, Fraction=@n56), B=@n54)) {name=GalaxiesAndStars, nodeColor=<rgba 0 1 0.9063 1>}
 ]}
